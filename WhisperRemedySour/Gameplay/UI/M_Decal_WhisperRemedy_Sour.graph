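# Translucent, unlit deferred decal material for WhisperRemedy Sour that uses a shield texture, stored under the gameplay UI folder.
# A time-driven sine wave is remapped from its -1 to 1 range into a partial-to-full range and multiplied with the texture's alpha, producing a continuously pulsing shield mask that never fades out completely.
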Material /WhisperRemedySour/Gameplay/UI/M_Decal_WhisperRemedy_Sour
graph {
 properties {
  bAutomaticallySetUsageInEditor=false
  bCastRayTracedShadows=false
  blendMode=BLEND_Translucent
  materialDomain=MD_DeferredDecal
  shadingModel=MSM_Unlit
  usedShadingModels=Unlit
 }
 nodes [
  n1 = Time()
  Multiply(
   A=TextureSample() {texture=&/T_WhisperRemedy_Sour_Shield}.A
   B=MaterialFunctionCall(Input=Sine(None=@n1), Input Low=Constant() {r=-1}, Input High=Constant() {r=1}, Target Low=Constant() {r=0.4}, Target High=Constant() {r=1}) {materialFunction=&RemapValueRange}.Result
  ) {constB=0.8}
 ]
}
